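# M_TacoBits is a masked, lit material for Niagara sprites in the Taco skydiving contrail cosmetic, drawing bits of taco.
# A diffuse texture is multiplied by particle color, so the emitter can tint each sprite.
# A normal map texture gives the bits surface detail under lighting.
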Material /BRCosmetics/Athena/Effects/Particles/Skydiving/NS_Contrail_Taco/M_TacoBits
graph {
 properties {bUsedWithNiagaraSprites=true, blendMode=BLEND_Masked, pixelDepthOffsetMode=PDOM_Legacy, usedShadingModels=Default Lit}
 nodes [
  Multiply(A=ParticleColor().RGB, B=TextureSample() {texture=&/T_TacoBits_D}.RGB)
  TextureSample() {samplerType=SAMPLERTYPE_Normal, texture=&/T_TacoBits_N}
 ]
}
